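; Settings actor for a minifig cosmetic turntable scene, controlling the camera, background and displayed minifig.
; Construction script applies background visibility to scene components when enabled, and assigns the chosen Minifig mesh part to the minifig actor's AssembledMeshUser.
; When enabled, BeginPlay sets the player's view target to the full-body or portrait camera per display mode, with a next-tick delay if the controller is invalid.
; It also hides background actors, then reapplies the Show Background setting.
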(event EventBeginPlay
  (bind _returnvalue (Game|GetPlayerController 0))
  (if (Variables|General|GetEnabled)
    (Utilities|FlowControl|Sequence
      (:then_0
        (Utilities|IsValid _returnvalue
          (:Is%20Valid
            _)
          (:Is%20Not%20Valid
            (Utilities|FlowControl|DelayUntilNextTick)))
        (CallFunction|InitializeCamera _returnvalue))
      (:then_1
        (CallFunction|ShowBackgroundActors false)
        (CallFunction|ShowBackgroundActors (Variables|General|GetShow%20Background))))))

(fn GetCamera
  (switch Utilities|FlowControl|Switch|SwitchonMinifigDisplayMode (Variables|General|GetDisplay%20Mode)
    (:NewEnumerator0
      (return (Variables|Cameras|GetMinifigFullBodyCamera)))
    (:NewEnumerator1
      (return (Variables|Cameras|GetMinifigPortraitCamera)))))

(fn InitializeCamera
  (params (Player%20Controller :type Player%20Controller%20Object%20Reference))
  (bind _camera (CallFunction|GetCamera))
  (Game|Player|SetViewTargetwithBlend Player%20Controller _camera 0.0 "VTBlend_Linear" 0.0 false)
  (Fortnite|FortShippingLog self (Utilities|String|Append "Set View Target = " (Utilities|GetDisplayName _camera))))

(fn ShowBackgroundActors
  (params (Visible :type Boolean))
  (for _array_element (Variables|Actors|GetBackgroundActors)
    (Rendering|SetActorHiddenInGame _array_element (not Visible))))

(fn ConstructionScript
  (bind _minifigactor (Variables|Actors|GetMinifigActor))
  (Utilities|FlowControl|Sequence
    (:then_0
      (if (Variables|General|GetEnabled)
        (for _array_element (Variables|Actors|GetBackgroundActors)
          (for _array_element_2 (Actor|GetComponentsByClass _array_element "/Script/Engine.SceneComponent")
            (Rendering|SetVisibility _array_element_2 (Variables|General|GetShow%20Background) false)))))
    (:then_1
      (Utilities|IsValid (Variables|General|GetMinifig)
        (:Is%20Valid
          (Utilities|IsValid _minifigactor
            (:Is%20Valid
              (Variables|/Script/AssembledMeshSystem.AssembledMeshUserComponent|SetMeshPart (Variables|General|GetMinifig) (Variables|GetAssembledMeshUser _minifigactor)))))))))
