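; TurntableCamera is a Fortnite camera Blueprint, apparently for a cosmetic turntable, that fixes the world lighting when it activates.
; On activation it sets the time of day to noon, stops time progression and makes the contextual time of day manager visible.
; The player controller passed to the event is unused, and the construction script is empty.
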(event Camera|Fortnite|EventOnActivated
  (params (PlayerController :type Fort%20Player%20Controller%20Object%20Reference))
  (TimeOfDay|SetTimeOfDay 12.0 :!class /Script/FortniteGame.FortKismetLibrary)
  (TimeOfDay|SetTimeOfDaySpeed 0.0 :!class /Script/FortniteGame.FortKismetLibrary)
  (TimeOfDay|SetTimeOfDayManagerVisibility (TimeOfDay|GetContextualTimeOfDayManager) true))

(fn ConstructionScript)
