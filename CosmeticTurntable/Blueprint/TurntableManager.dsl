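; Turntable manager Blueprint that stages cosmetic loadouts in a vault preview world and rotates them on screen.
; Each tick it turns the preview rotator by delay-adjusted time, as a looping full turn or a sine pan between angle limits.
; Setup events hide backdrop elements, add default and Juno minifig items to the loadout, and delay emotes by the preload time.
; Icon permutation helpers update companion mannequin variants and set the permutation camera.
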(event EventTick
  (params (DeltaSeconds :type Float%20%28single-precision%29))
  (bind _itempreviewrotator (Variables|TurntableManager|GetItemPreviewRotator))
  (bind _returnvalue (Class|TurntableItemPreviewRotator|GetRotatingComponent _itempreviewrotator))
  (Utilities|IsValid (Variables|TurntableManager|GetItemPreviewRotator)
    (:Is%20Valid
      (bind _returnvalue_2 (CallFunction|GetAdjustedTime))
      (switch Utilities|FlowControl|Switch|SwitchonECosmeticTurntableRotationMode (Utilities|Struct|BreakCosmeticTurntableSettings (Variables|TurntableSettings|GetTurntableSettings))
        (:Full
          (bind _newparam (CallFunction|GetTurntableRotation _returnvalue_2))
          (Variables|Default|SetTickRotation _newparam))
        (:Pan
          (bind _newparam_2 (CallFunction|GetPanRotation _returnvalue_2))
          (Variables|Default|SetTickRotation _newparam_2)))
      (Utilities|IsValid _returnvalue
        (:Is%20Valid
          (Transformation|SetWorldRotation _returnvalue (Variables|Default|GetTickRotation) false false))))))

(event EventDayManagerSetup
  ;[0.5015 1 0.9099] Show/Hide Background & Floor mesh
  ;;[1 1 1] Never valid!
  (bind _asfort_time_of_day_manager (Utilities|Casting|CastToFortTimeOfDayManager (TimeOfDay|GetGlobalTimeOfDayManager))
    (:then
      _)
    (:CastFailed
      (Utilities|FlowControl|DelayUntilNextTick)))
  (TimeOfDay|SetTimeOfDayManagerVisibility _asfort_time_of_day_manager false))

(event EventVaultWorldSetup
  ;[0.5015 1 0.9099] Show/Hide Background & Floor mesh
  (bind _returnvalue (Rendering|Material|CreateDynamicMaterialInstance "/Game/Characters/Player/Male/Medium/Bodies/M_MED_HighTower_Tomato_Casual/Materials/MI_VertexCrunch.MI_VertexCrunch" "None" "None" :!class /Script/Engine.KismetMaterialLibrary))
  (Class|FortItemPreviewWorld|SetFloorMaterialDynamic (Variables|TurntableManager|GetVaultWorld) _returnvalue)
  (bind _asturntable_vault_world (Utilities|Casting|CastToTurntableVaultWorld (Variables|TurntableManager|GetVaultWorld)))
  (bind _background (Class|TurntableVaultWorld|GetTurntableBackground _asturntable_vault_world))
  (Utilities|IsValid _background
    (:Is%20Valid
      _)
    (:Is%20Not%20Valid
      (Utilities|FlowControl|DelayUntilNextTick)))
  (Rendering|SetVisibility (Variables|/Script/Engine.StaticMeshActor|GetStaticMeshComponent _background) false true))

(event EventBackgroundActorsSetup
  ;[0.5015 1 0.9099] Show/Hide Background & Floor mesh
  (Utilities|FlowControl|Sequence
    (:then_0
      (for _array_element (Variables|Components|GetNo%20Background%20Post%20Process%20Volumes)
        (Variables|/Script/Engine.PostProcessVolume|SetbEnabled false _array_element)))
    (:then_1
      (for _array_element_2 (Variables|Components|GetBackground%20Post%20Process%20Volumes)
        (Variables|/Script/Engine.PostProcessVolume|SetbEnabled true _array_element_2)))
    (:then_2
      (CallFunction|ProtoShowBackgroundActors false))))

(event EventHandleOnPreviewShown
  (bind _itempreviewrotator (Variables|TurntableManager|GetItemPreviewRotator))
  (bind _returnvalue (Class|FortItemPreviewSceneChanger|GetFocusCameraTarget _itempreviewrotator))
  (|CallParentFunction)
  (Utilities|IsValid _itempreviewrotator
    (:Is%20Valid
      (if (Utilities|IsValid _returnvalue :!class /Script/Engine.KismetSystemLibrary)
        (bind _asbp_fort_item_vehicle_preview_actor_cinematic (Utilities|Casting|Badcastnode (Class|FortCameraTargetPedestal|GetDisplayedActor _returnvalue)))
        (Rendering|SetVisibility (Variables|GetExponentialHeightFog_0 _asbp_fort_item_vehicle_preview_actor_cinematic) false true)))))

(event EventSetEmoteDelay
  (params (OnPawnActor :type Fort%20Item%20Preview%20On%20Pawn%20Actor%20Object%20Reference))
  ;[1 1 1] Set delay for emote animation.
  (Utilities|IsValid OnPawnActor
    (:Is%20Valid
      (Variables|/Script/FortniteGame.FortItemPreviewOnPawnActor|SetEmotePlayDelay (Variables|PreviewSettings|GetPreloadDelay) OnPawnActor)
      (Variables|/Script/FortniteGame.FortItemPreviewOnPawnActor|SetbForceEmotePlayDelay true OnPawnActor))))

(event EventSetupDefaultItemDefinitions
  (params (LoadoutBuilder :type Cosmetic%20Turntable%20Loadout%20Builder%20Object%20Reference))
  (bind _minifigoverride (Variables|Loadout|GetMinifigOverride))
  (bind _baseathenacharacteritemdefinition (Variables|/Script/FigureCharacterNativeRuntime.JunoAthenaCharacterItemOverrideDefinition|GetBaseAthenaCharacterItemDefinition _minifigoverride))
  (bind _output (Utilities|ResolveSoftReference _baseathenacharacteritemdefinition))
  (bind _minifigemoteoverride (Variables|Loadout|GetMinifigEmoteOverride))
  (bind _baseathenadanceitemdefinition (Variables|/Script/FigureCharacterNativeRuntime.JunoAthenaDanceItemOverrideDefinition|GetBaseAthenaDanceItemDefinition _minifigemoteoverride))
  (bind _cosmeticitems (Variables|Loadout|GetCosmeticItems))
  (bind _cosmeticitems_2 (Variables|Loadout|GetCosmeticItems))
  (Utilities|FlowControl|Sequence
    (:then_0
      (if (Utilities|IsValid _minifigemoteoverride :!class /Script/Engine.KismetSystemLibrary)
        (Variables|/Script/CosmeticTurntableRuntime.CosmeticTurntableLoadoutBuilder|SetForcePlaylistOverrideTag "(TagName=\"Product.Juno\")" LoadoutBuilder)
        (Utilities|LoadAssetBlocking _baseathenadanceitemdefinition)
        (Utilities|Array|Insert (Variables|Loadout|GetCosmeticItems) (Utilities|ResolveSoftReference _baseathenadanceitemdefinition))))
    (:then_1
      (if (Utilities|IsValid _minifigoverride :!class /Script/Engine.KismetSystemLibrary)
        (Variables|/Script/CosmeticTurntableRuntime.CosmeticTurntableLoadoutBuilder|SetForcePlaylistOverrideTag "(TagName=\"Product.Juno\")" LoadoutBuilder)
        (Utilities|LoadAssetBlocking _baseathenacharacteritemdefinition)
        (if (Utilities|Array|IsNotEmpty _cosmeticitems)
          (Utilities|Casting|CastToAthenaDanceItemDefinition (Utilities|Array|Get(acopy) _cosmeticitems)
            (:then
              (Utilities|Array|Insert _cosmeticitems_2 _output 1))
            (:CastFailed
              _))
          (else
            _))
        (Utilities|Array|Insert _cosmeticitems_2 _output)))
    (:then_2
      (Utilities|Array|Add (Variables|Loadout|GetCosmeticItems) (Variables|Loadout|GetDefaultCharacter))
      (Utilities|Array|Add (Variables|Loadout|GetCosmeticItems) (Variables|Loadout|GetDefaultVehicleBody)))))

(event EventTryUpdateCompanionMannequin
  ;[0 0.3646 0.124 0.5] Icon Permutation
  (Utilities|FlowControl|DelayUntilNextTick)
  (bind _returnvalue (Actor|GetActorOfClass "/Script/CosmeticCompanionsRuntime.FortCompanionMannequin"))
  (Utilities|IsValid _returnvalue
    (:Is%20Valid
      _)
    (:Is%20Not%20Valid
      _))
  (CallFunction|UpdateCompanionMannequin _returnvalue))

(event EventTrySetPermutationCamera
  ;[0 0.3646 0.124 0.5] Icon Permutation
  (Utilities|FlowControl|DelayUntilNextTick)
  (bind _returnvalue (Actor|GetActorOfClass "/Script/LevelSequence.LevelSequenceActor"))
  (Utilities|IsValid _returnvalue
    (:Is%20Valid
      _)
    (:Is%20Not%20Valid
      _))
  (CallFunction|SetPermutationCamera (Variables|TurntableManager|GetPermutationCamera)))

(fn ComputePreloadDelay
  (bind _returnvalue (Sequencer|Sequence|GetDisplayRate (Variables|Default|GetTarget%20Level%20Sequence)))
  (return (* (Utilities|TimeManagement|FrameRateToInterval _returnvalue) (Variables|Settings|GetPreloadFrames))))

(fn CreatePreviewWidget
  (bind _returnvalue (UserInterface|CreateWidget (Variables|TurntableManager|GetPreviewWidgetClass)))
  (bind _aswbp_cosmetic_turntable (Utilities|Casting|CastToWBP_CosmeticTurntable _returnvalue))
  (bind _output_get (Variables|Default|SetCurrentPreviewWidget _aswbp_cosmetic_turntable))
  (Class|WBPCosmeticTurntable|SetInnerBoxPadding _output_get (Variables|PreviewSettings|GetInnerBoxMargin))
  (UserInterface|Viewport|AddtoViewport _output_get 0)
  (return (Variables|/CosmeticTurntable/Blueprint/WBP_CosmeticTurntable.WBP_CosmeticTurntable_C|GetSizeBox_CameraFrame (Variables|Default|GetCurrentPreviewWidget))))

(fn GetAdjustedTime
  (params (InFrameRate :type Frame%20Rate%20Structure))
  (return (Math|Float|Clamp(Float) (- (Utilities|Time|GetTimeSeconds) (Variables|PreviewSettings|GetPreloadDelay)) 0.0 (Utilities|Time|GetTimeSeconds))))

(fn GetBaseTemplateIdOfJunoOverride
  (params (AccountItemOverrideDefinition :type Object%20Reference))
  (Utilities|FlowControl|Sequence
    (:then_0
      (bind _asjuno_athena_character_item_override_definition (Utilities|Casting|CastToJunoAthenaCharacterItemOverrideDefinition AccountItemOverrideDefinition))
      (Utilities|LoadAssetBlocking (Variables|/Script/FigureCharacterNativeRuntime.JunoAthenaCharacterItemOverrideDefinition|GetBaseAthenaCharacterItemDefinition _asjuno_athena_character_item_override_definition))
      (return (ItemDefinition|GetTemplateId (Utilities|ResolveSoftReference (Variables|/Script/FigureCharacterNativeRuntime.JunoAthenaCharacterItemOverrideDefinition|GetBaseAthenaCharacterItemDefinition _asjuno_athena_character_item_override_definition)))))
    (:then_1
      (bind _asjuno_athena_dance_item_override_definition (Utilities|Casting|CastToJunoAthenaDanceItemOverrideDefinition AccountItemOverrideDefinition))
      (Utilities|LoadAssetBlocking (Variables|/Script/FigureCharacterNativeRuntime.JunoAthenaDanceItemOverrideDefinition|GetBaseAthenaDanceItemDefinition _asjuno_athena_dance_item_override_definition))
      (return (ItemDefinition|GetTemplateId (Utilities|ResolveSoftReference (Variables|/Script/FigureCharacterNativeRuntime.JunoAthenaDanceItemOverrideDefinition|GetBaseAthenaDanceItemDefinition _asjuno_athena_dance_item_override_definition)))))))

(fn GetPanRotation
  (params (InCurrentTime :type Float%20%28double-precision%29))
  (bind _turntablesettings (Variables|TurntableSettings|GetTurntableSettings))
  (bind (_rotationmode _rotationcycleseconds _rotationminangle _rotationmaxangle) (Utilities|Struct|BreakCosmeticTurntableSettings _turntablesettings))
  (bind _returnvalue (Math|Float|MultiplybyPi 2.0))
  (bind _returnvalue_2 (Math|Float|MapRangeUnclamped InCurrentTime 0.0 _rotationcycleseconds 0.0 _returnvalue))
  (bind _returnvalue_3 (Math|Trig|Sin(Radians) _returnvalue_2))
  (bind _returnvalue_4 (Math|Float|MapRangeClamped _returnvalue_3 -1.0 1.0 _rotationminangle _rotationmaxangle))
  (if (Variables|Default|GetDebug)
    (Development|PrintString (Utilities|String|JoinStringArray (Utilities|Array|MakeArray (Utilities|String|ToString(Float) InCurrentTime) (Utilities|String|ToString(Float) _returnvalue_2) (Utilities|String|ToString(Float) _returnvalue_3) (Utilities|String|ToString(Float) _returnvalue_4)) "   ") true true "(R=0.000000,G=0.660000,B=1.000000,A=1.000000)" 0.2)
    (else
      _))
  (return (Math|Rotator|MakeRotator 0.0 0.0 _returnvalue_4)))

(fn GetTurntableItemPreviewRotator
  (bind _asturntable_vault_world (Utilities|Casting|CastToTurntableVaultWorld (Variables|TurntableManager|GetVaultWorld)))
  (bind _vaultrotator (Class|TurntableVaultWorld|GetTurntableVaultRotator _asturntable_vault_world))
  (return _vaultrotator))

(fn GetTurntableRotation
  (params (InCurrentTime :type Float%20%28single-precision%29))
  (bind (_rotationmode _rotationcycleseconds _rotationminangle _rotationmaxangle) (Utilities|Struct|BreakCosmeticTurntableSettings (Variables|TurntableSettings|GetTurntableSettings)))
  (return (Math|Rotator|MakeRotator 0.0 0.0 (Math|Float|MapRangeClamped (Math|Float|Lerp (Math|Float|MakePulsatingValue InCurrentTime (Math|Float|Power _rotationcycleseconds -1.0)) (Math|Float|Wrap(Float) InCurrentTime 0.0 _rotationcycleseconds) 1.0) 0.0 (Math|Float|Lerp 1.0 _rotationcycleseconds 1.0) _rotationminangle _rotationmaxangle))))

(fn GetTurntableVaultRotator
  (bind _asturntable_vault_world (Utilities|Casting|CastToTurntableVaultWorld (Variables|TurntableManager|GetVaultWorld))
    (:then
      (bind _vaultrotator (Class|TurntableVaultWorld|GetTurntableVaultRotator _asturntable_vault_world))
      (return _vaultrotator))
    (:CastFailed
      (return 0))))

(fn ProtoShowBackgroundActors
  (params (Visible :type Boolean))
  (for _array_element (Variables|Components|GetProtoBackgroundActors)
    (Rendering|SetActorHiddenInGame _array_element (not Visible))))

(fn ShowLoadout
  (params (TurntableWorld :type Turntable%20Vault%20World%20Object%20Reference) (Builder :type Cosmetic%20Turntable%20Loadout%20Builder%20Object%20Reference))
  (bind _vaultrotator (Class|TurntableVaultWorld|GetTurntableVaultRotator TurntableWorld))
  (Class|TurntableItemPreviewRotator|ShowLoadout _vaultrotator Builder))

(fn UpdateCompanionMannequin
  (params (CompanionMannequin :type Fort%20Companion%20Mannequin%20Object%20Reference))
  (bind _returnvalue (Cosmetics|GetCosmeticCompanionItemDefinition CompanionMannequin))
  (bind (_output_templateid _output_bmainslot _output_variantoverrides) (Utilities|Struct|BreakLoadoutSlotOverrideData (Utilities|Array|Get(aref) (Variables|LoadoutSettings|GetLoadoutSlots))))
  (bind _companionmannequin (Variables|GetCompanionMannequin))
  (Utilities|IsValid CompanionMannequin
    (:Is%20Valid
      (Utilities|IsValid _returnvalue
        (:Is%20Valid
          (for _array_element (select (Utilities|Array|IsNotEmpty (Variables|LoadoutSettings|GetLoadoutSlots)) _output_variantoverrides (Variables|Loadout|GetVariantsData))
            (bind _returnvalue_2 (Class|CosmeticTurntableLoadoutBuilder|GenerateVariantChannelInfo _returnvalue _array_element))
            (Utilities|Array|Add (LocalVariables|GetOverrideVariantsLocal) _returnvalue_2))
          (Variables|/Script/CosmeticCompanionsRuntime.FortCompanionMannequin|SetOverrideVariants (LocalVariables|GetOverrideVariantsLocal) _companionmannequin)
          (Cosmetics|ReInitializeCustomization _companionmannequin true))))))

(fn ConstructionScript)
