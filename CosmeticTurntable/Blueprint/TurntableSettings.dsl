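; TurntableSettings configures a cosmetic turntable display, supplying time-based rotation and prototype helpers for minifig, character and vehicle modes.
; Pan yaw swings on a cosine between angle limits; turntable yaw sweeps linearly and wraps each cycle, as constant Lerp alphas bypass its pulsating value.
; The display mode selects the primary actor and instantly switched camera view.
; Show functions toggle visibility for background and per-mode actor arrays, individually or together.
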(fn GetPanRotation
  (return (Math|Rotator|MakeRotator 0.0 0.0 (Math|Float|MapRangeClamped (Math|Trig|Cos(Radians) (Math|Float|MapRangeUnclamped (Utilities|Time|GetTimeSeconds) 0.0 (Variables|Turntable|GetRotationCycleSeconds) 0.0 (Math|Float|MultiplybyPi 2.0))) -1.0 1.0 (Variables|Turntable|GetPanMinAngle) (Variables|Turntable|GetPanMaxAngle)))))

(fn GetTurntableRotation
  (return (Math|Rotator|MakeRotator 0.0 0.0 (Math|Float|MapRangeClamped (Math|Float|Lerp (Math|Float|MakePulsatingValue (Utilities|Time|GetTimeSeconds) (Math|Float|Power (Variables|Turntable|GetRotationCycleSeconds) -1.0)) (Math|Float|Wrap(Float) (Utilities|Time|GetTimeSeconds) 0.0 (Variables|Turntable|GetRotationCycleSeconds)) 1.0) 0.0 (Math|Float|Lerp 1.0 (Variables|Turntable|GetRotationCycleSeconds) 1.0) (Variables|Turntable|GetTurntableMinAngle) (Variables|Turntable|GetTurntableMaxAngle)))))

(fn ProtoGetPrimaryActor
  (switch Utilities|FlowControl|Switch|SwitchonTurntableProtoMode (Variables|Turntable|GetTurntableDisplayMode)
    (:NewEnumerator0
      (return (Utilities|Array|Get(acopy) (Variables|Proto|GetProtoMinifigActors))))
    (:NewEnumerator1
      (return (Utilities|Array|Get(acopy) (Variables|Proto|GetProtoCharacterActors))))
    (:NewEnumerator2
      (return (Utilities|Array|Get(acopy) (Variables|Proto|GetProtoVehicleActors))))))

(fn ProtoInitializeCamera
  (params (PlayerController :type Player%20Controller%20Object%20Reference))
  (switch Utilities|FlowControl|Switch|SwitchonTurntableProtoMode (Variables|Turntable|GetTurntableDisplayMode)
    (:NewEnumerator0
      (Game|Player|SetViewTargetwithBlend PlayerController (Variables|Proto|GetProtoMinifigCamera) 0.0 "VTBlend_Linear" 0.0 false))
    (:NewEnumerator1
      (Game|Player|SetViewTargetwithBlend PlayerController (Variables|Proto|GetProtoCharacterCamera) 0.0 "VTBlend_Linear" 0.0 false))
    (:NewEnumerator2
      (Game|Player|SetViewTargetwithBlend PlayerController (Variables|Proto|GetProtoVehicleCamera) 0.0 "VTBlend_Linear" 0.0 false))))

(fn ProtoShowAllActors
  (params (Visible :type Boolean))
  (Utilities|FlowControl|Sequence
    (:then_0
      (Proto|ProtoShowBackgroundActors Visible))
    (:then_1
      (Proto|ProtoShowMinifigActors Visible))
    (:then_2
      (Proto|ProtoShowCharacterActors Visible))
    (:then_3
      (Proto|ProtoShowVehicleActors Visible))))

(fn ProtoShowBackgroundActors
  (params (Visible :type Boolean))
  (for _array_element (Variables|Proto|GetProtoBackgroundActors)
    (Rendering|SetActorHiddenInGame _array_element (not Visible))))

(fn ProtoShowCharacterActors
  (params (Visible :type Boolean))
  (for _array_element (Variables|Proto|GetProtoCharacterActors)
    (Rendering|SetActorHiddenInGame _array_element (not Visible))))

(fn ProtoShowMinifigActors
  (params (Visible :type Boolean))
  (for _array_element (Variables|Proto|GetProtoMinifigActors)
    (Rendering|SetActorHiddenInGame _array_element (not Visible))))

(fn ProtoShowVehicleActors
  (params (Visible :type Boolean))
  (for _array_element (Variables|Proto|GetProtoVehicleActors)
    (Rendering|SetActorHiddenInGame _array_element (not Visible))))

(fn ConstructionScript)
