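# Editor Utility Widget for the CosmeticTurntable plugin that launches cosmetic turntable renders and asset exports from the editor.
# Render Turntable runs EditorScript_RenderCosmetics.py through Execute Python Command, with a JSON config built from the typed file name, or tt_mix2 when empty.
# Export Assets runs the same script with a fixed export config.
# All paths are hard-coded local drive paths; RenderTurntable targets another workspace, and BeginMovieRender is an empty stub.
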EditorUtilityWidgetBlueprint /CosmeticTurntable/Blueprint/TurntableUtilityWidget
graphs [
 BeginMovieRender {nodes [BeginMovieRender!() -> (
    then
    Target Level Sequence:Level Sequence Object Reference
    Target Map:Soft Object Path Structure
    Render Dimensions:Int Point Structure
    Preload Frames:int
    OnRenderFinished:"Delegate (by ref)"
   ) {is=title K2Node_FunctionEntry}]}
 EventGraph {nodes [
   "On Clicked (Button_RenderTurntable)"() {is=title K2Node_ComponentBoundEvent}
   n1 = Get Text(self=Get Input_TurntableFileName() -> (Input_TurntableFileName:Editor Utility Editable Text Box Object Reference))
   > n2 = Branch(Condition=TextIsEmpty(InText=@n1))
   > "Execute Python Command (Advanced)"!(
    execute
    PythonCommand:String="F:\\FN-Release-B\\FortniteGame\\Plugins\\CosmeticTurntable\\Content\\Python\\EditorScript_RenderCosmetics.py -C F:\\Cosmetics\\tt_mix2.json "
    ExecutionMode:EPythonCommandExecutionMode Enum=ExecuteFile
    FileExecutionScope:EPythonFileExecutionScope Enum=Private
   ) -> (then, CommandResult:String, LogOutput:Array of Python Log Output Entry Structures, ReturnValue:bool)
   "Execute Python Command (Advanced)"(
    execute=@n2.else
    PythonCommand="To String (Text)"(InText=Format Text!(
      Format="LOCTEXT(\"F:\\\\FN-Release-B\\\\FortniteGame\\\\Plugins\\\\CosmeticTurntable\\\\Content\\\\Python\\\\EditorScript_RenderCosmetics.py -C F:\\\\Cosmetics\\\\{TurntableFileName}.json \")"
      TurntableFileName:Text=@n1
     ) -> (Result))
    ExecutionMode=ExecuteFile
    FileExecutionScope=Private
   )
   "On Clicked (Button_ExportAssets)"() {is=title K2Node_ComponentBoundEvent}
   > "Execute Python Command (Advanced)"(
    PythonCommand="F:\\FN-Release-B\\FortniteGame\\Plugins\\CosmeticTurntable\\Content\\Python\\EditorScript_RenderCosmetics.py -C F:\\Cosmetics\\assets_to_export.json"
    ExecutionMode=ExecuteFile
    FileExecutionScope=Private
   )
   Event PreConstruct()
   Event Construct()
   Event Tick()
  ]}
 RenderTurntable {nodes [
   RenderTurntable() {is=title K2Node_FunctionEntry}
   > "Execute Python Command (Advanced)"(
    PythonCommand="F:\\Asil.Karatas_FN-Release\\FortniteGame\\Plugins\\CosmeticTurntable\\Content\\Python\\EditorScript_RenderCosmetics.py -C F:\\Cosmetics\\tt_mix.json -s 0 -e 1"
    ExecutionMode=ExecuteFile
    FileExecutionScope=Private
   )
  ]}
]
variables [
 NewVar {type=string, default=""}
 Preload Frames {type=integer}
 Render Dimensions {type=IntPoint, default=<0 0>}
 Target Level Sequence {type=LevelSequence, default=null}
 Target Map {type=SoftObjectPath, default=""}
 TurntableManager {type=TurntableManager_C, default=null}
]
widgets {
 info {rootWidgetClass=&VerticalBox, widgetCount=9}
 widgets [
  VerticalBox_0 {widgetClassPath=&VerticalBox, widget=VerticalBox_0}
  HorizontalBox_0 {widgetClassPath=&HorizontalBox, parent=VerticalBox_0, widget=HorizontalBox_0, slot {class=VerticalBoxSlot}}
  Label_TurntableFileName {
   widgetClassPath=&TextBlock
   parent=HorizontalBox_0
   widget=Label_TurntableFileName
   properties {text=Turntable File Name:}
   slot {class=HorizontalBoxSlot}
  }
  Input_TurntableFileName {
   widgetClassPath=&EditorUtilityEditableTextBox
   bIsVariable=true
   parent=HorizontalBox_0
   widget=Input_TurntableFileName
   slot {class=HorizontalBoxSlot, properties {size {sizeRule=Fill, value=1}}}
  }
  Button_RenderTurntable {widgetClassPath=&EditorUtilityButton, parent=VerticalBox_0, widget=Button_RenderTurntable, slot {class=VerticalBoxSlot}}
  TextBlock_128 {
   widgetClassPath=&TextBlock
   parent=Button_RenderTurntable
   widget=TextBlock_128
   properties {text=Render Turntable}
   slot {class=ButtonSlot}
  }
  VerticalBox_1 {widgetClassPath=&VerticalBox, parent=VerticalBox_0, widget=VerticalBox_1, slot {class=VerticalBoxSlot}}
  Button_ExportAssets {widgetClassPath=&EditorUtilityButton, parent=VerticalBox_0, widget=Button_ExportAssets, slot {class=VerticalBoxSlot}}
  TextBlock_1 {
   widgetClassPath=&TextBlock
   parent=Button_ExportAssets
   widget=TextBlock_1
   properties {text=Export Assets}
   slot {class=ButtonSlot}
  }
 ]
}
