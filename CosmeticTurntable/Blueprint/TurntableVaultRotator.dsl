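; Item preview turntable scene that drives camera transitions, zoom, pedestal rotation, lighting and background for displayed items.
; Two half-turn timelines, played forward or reversed, interpolate camera transform and FOV and rotate lights, with directional sounds.
; Lighting switches between PC, low-detail and mobile/Switch setups by platform and shadow quality, refreshing on settings changes.
; Background colors use material parameter collection values; message text is a widget drawn to a render target.
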(event EventBeginPlay
  (if (Variables|Default|GetAlwaysOn)
    (Development|PrintString "ALWAYS ON")
    (else
      (Utilities|FlowControl|Sequence
        (:then_0
          ;[0 0 0 1 12] Force Disable On Start Up
          (CallFunction|SwitchPCLighting false)
          (CallFunction|SwitchMobileLighting false)
          (CallFunction|SwitchPCLightingLOWDetail false))
        (:then_1
          ;[0 0 0 1 12] Switch lighting if settings are update while screen is active
          (bind _returnvalue (Settings|GetGameUserSettings))
          (EventDispatchers|BindEventtoOnGameUserSettingsUINeedsUpdate _returnvalue (delegate Custom|UpdateSettings)))))))

(event Custom|UpdateSettings
  (bind _returnvalue (TouchPlatform|IsMobileGame))
  (bind _returnvalue_2 (TouchPlatform|IsNintendoSwitch))
  ;[0 0 0 1 12] Switch lighting if settings are update while screen is active
  (if (Variables|Default|GetIsActive)
    (if (Math|Boolean|ORBoolean (== (Settings|GetShadowQuality (stale EventBeginPlay _returnvalue)) 0) _returnvalue _returnvalue_2)
      (if (or _returnvalue _returnvalue_2)
        ;;[0.125 0.125 0.125 1 12] Force Mobile
        (CallFunction|SwitchPCLighting false)
        (CallFunction|SwitchPCLightingLOWDetail false)
        (CallFunction|SwitchMobileLighting true)
        (else
          ;;[0.125 0.125 0.125 1 12] Force PC Low Detail
          (CallFunction|SwitchPCLighting false)
          (CallFunction|SwitchMobileLighting false)
          (CallFunction|SwitchPCLightingLOWDetail true)))
      (else
        ;;[0.125 0.125 0.125 1 12] Force PC Lighting
        (CallFunction|SwitchMobileLighting false)
        (CallFunction|SwitchPCLightingLOWDetail false)
        (CallFunction|SwitchPCLighting true)))))

(event EventOnCameraTransitionReady
  (params (bPrimaryToSecondary :type Boolean) (PrimaryRequestedItem :type Fort%20Account%20Item%20Definition%20Object%20Reference) (Options :type Scene%20Transition%20Options%20Structure%20%28by%20ref%29))
  (bind _cameraturn_0_to_180 (Variables|TurntableVaultRotator|GetCameraTurn_0_to_180))
  (bind _returnvalue (Components|Timeline|GetPlaybackPosition _cameraturn_0_to_180))
  (bind _returnvalue_2 (Components|Timeline|GetTimelineLength _cameraturn_0_to_180))
  (bind _returnvalue_3 (- _returnvalue_2 _returnvalue))
  (bind _cameraturn_180_to_360 (Variables|TurntableVaultRotator|GetCameraTurn_180_to_360))
  (bind _returnvalue_4 (Components|Timeline|GetPlaybackPosition _cameraturn_180_to_360))
  (bind _returnvalue_5 (Components|Timeline|GetTimelineLength _cameraturn_180_to_360))
  (bind _returnvalue_6 (- _returnvalue_5 _returnvalue_4))
  (bind _transitioncameracomponent (Variables|FortItemPreviewSceneChanger|GetTransitionCameraComponent))
  (bind _transitioncameracomponent_2 (Variables|FortItemPreviewSceneChanger|GetTransitionCameraComponent))
  (Utilities|FlowControl|Delay 0.1)
  (switch Utilities|FlowControl|Switch|SwitchonESceneTransitionType (Utilities|Struct|BreakSceneTransitionOptions Options)
    (:NoTransition
      (CallFunction|CameraTransitionComplete))
    (:Clockwise
      (if bPrimaryToSecondary
        ;[0.1885 0.1885 0.1885] These sounds were originally intended for claim star going away but should be able to be reused on other contexts
        (CallFunction|PlaySoundwhenEnabled (Variables|Sound|GetClockwiseSound))
        (bind (_direction _fov _directionalrotation _camerapositionoffset _camerarotationoffset) (Timeline|CameraTurn_0_to_180 PlayFromStart)
          (:Update
            (Camera|SetFieldOfView _transitioncameracomponent_2 (.FieldOfView (Utilities|Struct|BreakInterpolatedTransitionCamera (CallFunction|GetInterpolatedTransitionCameraPosition (Math|Transform|MakeTransform _camerapositionoffset (Math|Rotator|MakeRotator (.x _camerarotationoffset) (.y _camerarotationoffset) (.z _camerarotationoffset))) _fov (Utilities|Select :Index _direction :Forward _returnvalue :Backward _returnvalue_3) _returnvalue_2))))
            (Transformation|SetRelativeTransform _transitioncameracomponent_2 (Utilities|Struct|BreakInterpolatedTransitionCamera (CallFunction|GetInterpolatedTransitionCameraPosition (Math|Transform|MakeTransform _camerapositionoffset (Math|Rotator|MakeRotator (.x _camerarotationoffset) (.y _camerarotationoffset) (.z _camerarotationoffset))) _fov (Utilities|Select :Index _direction :Forward _returnvalue :Backward _returnvalue_3) _returnvalue_2)) false true)
            (Transformation|SetRelativeRotation (Variables|Default|GetDirectionalLights) (Math|Rotator|MakeRotator 0.0 0.0 _directionalrotation) false false))
          (:Finished
            (CallFunction|CameraTransitionComplete)))
        (else
          (CallFunction|PlaySoundwhenEnabled (Variables|Sound|GetCounterClockwiseSound))
          (bind (_direction_2 _fov_2 _directionrotation _camerapositionoffset_2 _camerarotationoffset_2) (Timeline|CameraTurn_180_to_360 PlayFromStart)
            (:Update
              (Camera|SetFieldOfView _transitioncameracomponent (.FieldOfView (Utilities|Struct|BreakInterpolatedTransitionCamera (CallFunction|GetInterpolatedTransitionCameraPosition (Math|Transform|MakeTransform _camerapositionoffset_2 (Math|Rotator|MakeRotator (.x _camerarotationoffset_2) (.y _camerarotationoffset_2) (.z _camerarotationoffset_2))) _fov_2 (Utilities|Select :Index _direction_2 :Forward _returnvalue_4 :Backward _returnvalue_6) _returnvalue_5))))
              (Transformation|SetRelativeTransform _transitioncameracomponent (Utilities|Struct|BreakInterpolatedTransitionCamera (CallFunction|GetInterpolatedTransitionCameraPosition (Math|Transform|MakeTransform _camerapositionoffset_2 (Math|Rotator|MakeRotator (.x _camerarotationoffset_2) (.y _camerarotationoffset_2) (.z _camerarotationoffset_2))) _fov_2 (Utilities|Select :Index _direction_2 :Forward _returnvalue_4 :Backward _returnvalue_6) _returnvalue_5)) false true)
              (Transformation|SetRelativeRotation (Variables|Default|GetDirectionalLights) (Math|Rotator|MakeRotator 0.0 0.0 _directionrotation) false false))
            (:Finished
              (CallFunction|CameraTransitionComplete))))))
    (:CounterClockwise
      (if bPrimaryToSecondary
        (CallFunction|PlaySoundwhenEnabled (Variables|Sound|GetClockwiseSound))
        (Timeline|CameraTurn_180_to_360 ReverseFromEnd)
        (else
          (CallFunction|PlaySoundwhenEnabled (Variables|Sound|GetCounterClockwiseSound))
          (Timeline|CameraTurn_0_to_180 ReverseFromEnd))))))

(event EventOnTargetZoomLevelSet
  (params (TargetZoomLevel :type Float%20%28single-precision%29))
  (bind _output_get (Variables|Default|SetTargetZoomLevel TargetZoomLevel))
  (Math|Float|CompareFloat _output_get (Variables|FortItemPreviewSceneChanger|GetZoomLevel)
    (:>
      (bind (_direction _zoomlevel) (Timeline|Timeline_Zoom Play)
        (:Update
          (Utilities|FlowControl|Sequence
            (:then_0
              (switch Utilities|FlowControl|Switch|SwitchonETimelineDirection _direction
                (:Forward
                  (if (not (>= _zoomlevel (Variables|Default|GetTargetZoomLevel)))
                    (Variables|FortItemPreviewSceneChanger|SetZoomLevel _zoomlevel :!label _label)))
                (:Backward
                  (if (not (<= _zoomlevel (Variables|Default|GetTargetZoomLevel)))
                    (goto _label))))
              (Components|Timeline|Stop (Variables|TurntableVaultRotator|GetTimeline_Zoom))
              (Variables|FortItemPreviewSceneChanger|SetZoomLevel (Variables|Default|GetTargetZoomLevel)))
            (:then_1
              _)))
        (:Finished
          (Variables|FortItemPreviewSceneChanger|SetZoomLevel (Variables|Default|GetTargetZoomLevel)))))
    (:==
      _)
    (:<
      (Timeline|Timeline_Zoom Reverse)))
  (CallFunction|TickZoomLevel))

(event EventOnTargetZoomLevelCanceled
  (Timeline|Timeline_Zoom Stop))

(event EventOnTargetRotationOffsetSet
  (params (TargetRotationOffset :type Rotator%20%28by%20ref%29 :default "0, 0, 0"))
  (bind _output_get (Variables|Default|SetTargetRotationOffset TargetRotationOffset))
  (bind _output_get_2 (Variables|Default|SetStartingRotationOffset (Variables|FortItemPreviewSceneChanger|GetPedestalRotationOffset)))
  (if (not (Utilities|Operators|Equal(==) _output_get_2 _output_get 0))
    (bind (_direction _yawinterpalpha) (Timeline|Timeline_YawRotation PlayFromStart)
      (:Update
        (Variables|FortItemPreviewSceneChanger|SetPedestalRotationOffset (Math|Interpolation|Ease "Linear" _yawinterpalpha (Variables|Default|GetStartingRotationOffset) (Variables|Default|GetTargetRotationOffset) true))
        (CallFunction|TickPedestalRotationOffset))
      (:Finished
        (Variables|FortItemPreviewSceneChanger|SetPedestalRotationOffset (Variables|Default|GetTargetRotationOffset)))))
  (CallFunction|TickPedestalRotationOffset))

(event EventOnTargetRotationOffsetCanceled
  (Timeline|Timeline_YawRotation Stop))

(event EventOnSetBackgroundMessageText
  (params (BackgroundMessageText :type Text%20%28by%20ref%29))
  (CallFunction|SetBackgroundString BackgroundMessageText))

(event EventOnUpdateBackgroundColor
  (params (bEnableAutotestBackground :type Boolean))
  (if bEnableAutotestBackground
    (CallFunction|SetBackgroundColor "(R=0.000000,G=0.346704,B=0.000000,A=1.000000)" "(R=0.000000,G=0.346704,B=0.000000,A=1.000000)")
    (else
      (CallFunction|SetBackgroundColor "(R=0.030000,G=0.052500,B=0.300000,A=1.000000)" "(R=0.047500,G=0.348334,B=0.950000,A=1.000000)"))))

(fn LightControl
  (params (Active :type Boolean))
  (bind _returnvalue (TouchPlatform|IsMobileGame))
  (if (not (Variables|Default|GetAlwaysOn))
    (Variables|Default|SetIsActive Active)
    (bind _returnvalue_2 (Settings|GetGameUserSettings))
    (if (Math|Boolean|ORBoolean (== (Settings|GetShadowQuality _returnvalue_2) 0) (TouchPlatform|IsNintendoSwitch) _returnvalue)
      (if (or (TouchPlatform|IsNintendoSwitch) _returnvalue)
        (CallFunction|SwitchMobileLighting Active)
        (else
          (CallFunction|SwitchPCLightingLOWDetail Active)))
      (else
        (CallFunction|SwitchPCLighting Active)))))

(fn PlaySoundWhenEnabled
  (params (Sound :type Sound%20Base%20Object%20Reference))
  (if (CallFunction|IsSoundEnabled)
    (Audio|PlaySound2D Sound)))

(fn SetBackgroundColor
  (params (RGBA0 :type Linear%20Color%20Structure) (RGBA1 :type Linear%20Color%20Structure))
  (Rendering|Material|SetVectorParameterValue "/Game/UI/Foundation/Environment/Shell/Materials/M_UI_SceneBackgroundMCP.M_UI_SceneBackgroundMCP" "RGBA0" RGBA0 :!class /Script/Engine.KismetMaterialLibrary)
  (Rendering|Material|SetVectorParameterValue "/Game/UI/Foundation/Environment/Shell/Materials/M_UI_SceneBackgroundMCP.M_UI_SceneBackgroundMCP" "RGBA1" RGBA1 :!class /Script/Engine.KismetMaterialLibrary))

(fn SetBackgroundString
  (params (CustomText :type Text))
  (if (Utilities|Text|TextIsEmpty CustomText)
    (CallFunction|ToggleBackgroundText false)
    (else
      (bind _returnvalue (UserInterface|CreateWidget "/Game/Environments/FrontEnd/Blueprints/BackgroundString/UI_Background_Strings.UI_Background_Strings_C" 0 CustomText))
      (Rendering|DrawWidgettoRenderTarget _returnvalue "/Game/Environments/FrontEnd/Blueprints/BackgroundString/UI_BackgroundString.UI_BackgroundString")
      (CallFunction|ToggleBackgroundText true))))

(fn SwitchMobileLighting
  (params (Visibility :type Boolean))
  (Rendering|SetVisibility (Variables|Default|GetSkyLight) false false)
  (Rendering|SetVisibility (Variables|Default|GetDirectionalLightMobile) Visibility false)
  (Variables|/Script/Engine.PostProcessComponent|SetbEnabled Visibility (Variables|Default|GetPostProcess_Mobile)))

(fn SwitchPCLighting
  (params (Visibility :type Boolean))
  (Rendering|SetVisibility (Variables|Default|GetSkyLight) Visibility false))

(fn SwitchPCLighting_LOWDetail
  (params (Visibility :type Boolean))
  (Rendering|SetVisibility (Variables|Default|GetDirectionalLight_LOWPC) Visibility false)
  (Variables|/Script/Engine.PostProcessComponent|SetbEnabled Visibility (Variables|Default|GetPostProcess_LOWPC)))

(fn ToggleBackgroundText
  (params (bDisplayText :type Boolean))
  (if bDisplayText
    (Rendering|Material|SetScalarParameterValue "/Game/UI/Foundation/Environment/Shell/Materials/M_UI_SceneBackgroundMCP.M_UI_SceneBackgroundMCP" "bNoText" :!class /Script/Engine.KismetMaterialLibrary)
    (else
      (Rendering|Material|SetScalarParameterValue "/Game/UI/Foundation/Environment/Shell/Materials/M_UI_SceneBackgroundMCP.M_UI_SceneBackgroundMCP" "bNoText" 1.0 :!class /Script/Engine.KismetMaterialLibrary))))

(fn ConstructionScript)
