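; TurntableVaultWorld is a cosmetic turntable preview world Blueprint that exposes typed accessors to the scene actors of a FortItemPreviewWorld.
; GetTurntableBackground and GetTurntableVaultRotator cast the child actors of the scene background and scene changer components to FortStaticMeshActor and TurntableVaultRotator.
; GetVaultRotator casts the scene changer's child to TurntableItemPreviewRotator but returns an empty value instead of the cast result.
; BeginPlay and the construction script only call their parents; Tick and overlap events are empty.
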(event EventBeginPlay
  (|CallParentFunction))

(event Collision|EventActorBeginOverlap
  (params (OtherActor :type Actor%20Object%20Reference)))

(event EventTick
  (params (DeltaSeconds :type Float%20%28single-precision%29)))

(fn GetTurntableBackground
  (bind _asfort_static_mesh_actor (Utilities|Casting|CastToFortStaticMeshActor (Variables|/Script/Engine.ChildActorComponent|GetChildActor (Variables|FortItemPreviewWorld|GetSceneBackgroundComponent))))
  (return _asfort_static_mesh_actor))

(fn GetTurntableVaultRotator
  (bind _asturntable_vault_rotator (Utilities|Casting|CastToTurntableVaultRotator (Variables|/Script/Engine.ChildActorComponent|GetChildActor (Variables|FortItemPreviewWorld|GetSceneChangerComponent))))
  (return _asturntable_vault_rotator))

(fn GetVaultRotator
  (Utilities|Casting|CastToTurntableItemPreviewRotator (Variables|/Script/Engine.ChildActorComponent|GetChildActor (Variables|FortItemPreviewWorld|GetSceneChangerComponent)))
  (return 0))

(fn ConstructionScript
  (|CallParentFunction))
