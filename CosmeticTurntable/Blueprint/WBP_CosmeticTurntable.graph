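# WBP_CosmeticTurntable is a layout widget for a cosmetic turntable, apparently defining output and camera framing areas.
# It nests three SizeBoxes: an aspect-ratio-constrained viewport, an output frame with square size values, and an unevenly padded camera frame.
# The Set Inner Box Padding function casts the camera frame's slot to SizeBoxSlot and applies a given margin.
# The event graph holds only unconnected event stubs.
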WidgetBlueprint /CosmeticTurntable/Blueprint/WBP_CosmeticTurntable
graphs [
 EventGraph {nodes [
   Event PreConstruct()
   Event Construct()
   Event Tick()
  ]}
 Set Inner Box Padding {nodes [
   n1 = Set Inner Box Padding!() -> (then, InPadding:Margin Structure) {is=title K2Node_FunctionEntry}
   > n2 = Cast To SizeBoxSlot!(
    execute
    Object=Get Slot!(self:Widget Object Reference=Get SizeBox_CameraFrame() -> (SizeBox_CameraFrame:Size Box Object Reference)) -> (Slot:Panel Slot Object Reference)
   ) -> (then, CastFailed, AsSize Box Slot:Size Box Slot Object Reference) {is=title K2Node_DynamicCast}
   > SetPadding!(execute, self:Size Box Slot Object Reference=@n2, InPadding:Margin Structure=@n1) -> (then)
  ]}
]
variables []
widgets {
 info {inheritedWidgetCount=0, parentClass=&UserWidget, rootWidgetClass=&SizeBox, widgetCount=3}
 widgets [
  SizeBox_Viewport {
   widgetClassPath=&SizeBox
   bIsVariable=true
   widget=SizeBox_Viewport
   properties {bOverride_MaxAspectRatio=true, bOverride_MinAspectRatio=true}
  }
  SizeBox_OutputFrame {
   widgetClassPath=&SizeBox
   bIsVariable=true
   parent=SizeBox_Viewport
   widget=SizeBox_OutputFrame
   properties {heightOverride=1260, widthOverride=1260}
   slot {class=SizeBoxSlot}
  }
  SizeBox_CameraFrame {
   widgetClassPath=&SizeBox
   bIsVariable=true
   parent=SizeBox_OutputFrame
   widget=SizeBox_CameraFrame
   slot {class=SizeBoxSlot, properties {padding=<ltrb 50 250 50 10>}}
  }
 ]
}
