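# Default Lit material for a prototype locker turntable (BuffBass), with a two-color backdrop gradient and floor shading controls; no connections to material outputs appear.
# A screen-aligned radial gradient lerps Light to Accent into a named reroute, alongside brightness multipliers.
# Its aspect term divides resolution X by itself, so it applies no correction.
# UV radial masks blend floor color, specular, roughness and metallic, while saturated UV ramps subtract into another mask.
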Material /CosmeticTurntable/Materials/M_TurntableProtoLocker_BuffBass
graph {
 properties {pixelDepthOffsetMode=PDOM_Legacy, usedShadingModels=Default Lit}
 nodes [
  Constant()
  Constant() {r=1}
  n1 = TextureCoordinate()
  n2 = AppendVector(A=ScalarParameter() {defaultValue=-0.3507, parameterName=Param_4}, B=ScalarParameter() {defaultValue=-0.2813, parameterName=Param_5})
  n3 = OneMinus(None=MaterialFunctionCall(
    UVs=Add(A=@n1, B=@n2)
    Radius=ScalarParameter() {defaultValue=0.6173, parameterName=Floor-Radius Mask}
    Density=ScalarParameter() {defaultValue=4, parameterName=Floor-Density Mask}
   ) {materialFunction=&RadialGradientExponential}.RadialGradientExponential)
  LinearInterpolate(A=ScalarParameter() {defaultValue=0.05, parameterName=Roughness}, Alpha=@n3) {constA=1}
  n4 = MaterialFunctionCall(
   UVs=TextureCoordinate()
   Radius=ScalarParameter() {defaultValue=0.6173, parameterName=Floor-Radius Mask}
   Density=ScalarParameter() {defaultValue=4, parameterName=Floor-Density Mask}
  ) {materialFunction=&RadialGradientExponential}
  LinearInterpolate(B=VectorParameter() {parameterName=FloorRGB, defaultValue=<rgba 0.9647 0.9301 0.5089 1>}.RGB, Alpha=@n4.RadialGradientExponential)
  LinearInterpolate(B=ScalarParameter() {defaultValue=4, group=IsFloor, parameterName=Floor-Specular}, Alpha=@n4.RadialGradientExponential)
  LinearInterpolate(B=ScalarParameter() {defaultValue=1, group=IsFloor, parameterName=Floor-Roughness}, Alpha=@n4.RadialGradientExponential) {constA=1}
  LinearInterpolate(A=ScalarParameter() {defaultValue=0.6, parameterName=Metalllic}, Alpha=@n3) {constB=0}
  Multiply(A=NamedRerouteUsage(), B=ScalarParameter() {defaultValue=10})
  Multiply(
   A=Multiply(A=NamedRerouteUsage(), B=ScalarParameter() {defaultValue=1, parameterName=DesiredBrightness})
   B=ScalarParameter() {defaultValue=0.4, parameterName=Param_1}
  )
  n5 = MaterialFunctionCall(Float2=MaterialFunctionCall() {materialFunction=&ScreenResolution}.Visible Resolution) {materialFunction=&BreakOutFloat2Components}
  n6 = Divide(A=@n5.R, B=@n5.R)
  n7 = Add(
   A=AppendVector(A=Multiply(A=@n6) {constB=0.5}, B=Constant() {r=0.5})
   B=AppendVector(
    A=ScalarParameter() {defaultValue=0.65, group=Radial Gradient, parameterName=x-Offset}
    B=ScalarParameter() {defaultValue=0.5, group=Radial Gradient, parameterName=y-Offset}
   )
  )
  n8 = ScalarParameter() {defaultValue=1.8, desc="Only works if CrewLegacyUnsubscribed = 0", group=Radial Gradient, parameterName=Radius}
  n9 = ScalarParameter() {defaultValue=1, desc="Only works if CrewLegacyUnsubscribed = 0", group=Radial Gradient, parameterName=Density}
  n10 = Constant() {r=1}
  n11 = Multiply(
   A=Multiply(
    A=AppendVector(A=@n6, B=@n10)
    B=MaterialFunctionCall() {materialFunction=&ScreenAlignedUVs}."X 100%, Y 100%"
   )
   B=ScalarParameter() {defaultValue=1, group=Radial Gradient, parameterName=Tile}
  )
  NamedRerouteDeclaration(None=LinearInterpolate(
    A=VectorParameter() {parameterName=Light, defaultValue=<rgba 0.0166 0.0661 0.1927 1>}.RGB
    B=VectorParameter() {parameterName=Accent, defaultValue=<rgba 0.3874 0 0.7344 1>}.RGB
    Alpha=MaterialFunctionCall(UVs=@n11, CenterPosition=@n7, Radius=@n8, Density=@n9) {materialFunction=&RadialGradientExponential}.RadialGradientExponential
   )) {name=RGB Gradient Reroute, desc="Combined RGB0, RGB1 with a Radial Gradient in the Alpha", nodeColor=<rgba 0.6406 0 1 1>}
  n12 = TextureCoordinate()
  OneMinus(None=@n12)
  ScalarParameter() {defaultValue=0.25, parameterName=Specular}
  ScalarParameter() {group=IsFloor, parameterName=Floor-Metallic}
  n13 = Multiply(A=Add(A=@n12, B=ScalarParameter() {defaultValue=-0.397, parameterName=Param_3}), B=ScalarParameter() {defaultValue=4, parameterName=Param_2})
  n14 = TextureCoordinate()
  n15 = Multiply(
   A=Add(A=OneMinus(None=@n14), B=ScalarParameter() {defaultValue=-0.5197, parameterName=Param_3sd})
   B=ScalarParameter() {defaultValue=6, parameterName=Param_2sdcs}
  )
  Subtract(A=ComponentMask(None=Saturate(None=@n13)) {g=true}, B=ComponentMask(None=Saturate(None=@n15)) {r=true})
 ]
}
