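# Post-process material that paints a procedural backdrop over distant scene pixels for the cosmetic turntable's proto scene.
# Three MF_UI_Circle glows in screen-scaled UVs, sharpened by powers, blend four colors into a radial gradient.
# Normalized scene depth minus a threshold masks the gradient, with Scene_Terrain_Target's channels selecting the BR, Delmar or Juno threshold.
# Scene_Discovery_Transition moves that threshold toward the selected value while fading the effect in over the original image.
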Material /CosmeticTurntable/Materials/PPM_TurntableProtoScene
graph {
 properties {materialDomain=MD_PostProcess, pixelDepthOffsetMode=PDOM_Legacy, usedShadingModels=Default Lit}
 nodes [
  n1 = SceneTexture() {sceneTextureId=PPI_PostProcessInput0}
  n2 = VectorParameter() {parameterName=RGB-2, defaultValue=<rgba 0 0.002 0.1163 1>}
  n3 = VectorParameter() {parameterName=RGB-1, defaultValue=<rgba 0.2034 0.0076 0.3663 1>}
  n4 = Add(
   A=MaterialFunctionCall!(ScreenUV, TextureSize, ScalePivotPoint) {materialFunction=/FortUILibrary/MaterialFunctions/Basics/MF_UI_ScreenSpaceUVScaling}.Scaled to Fill
   B=AppendVector(A=ScalarParameter() {parameterName=s}, B=ScalarParameter() {parameterName=Param_3})
  )
  n5 = ScalarParameter() {defaultValue=0.5}
  n6 = ScalarParameter() {parameterName=Param_2}
  n7 = ScalarParameter() {defaultValue=1, parameterName=Param_1}
  n8 = Saturate(None=Power(
    Base=MaterialFunctionCall(UVs=@n4, Size=@n5, GlowMin=@n6, GlowMax=@n7) {materialFunction=&MF_UI_Circle}.Fill
    Exp=ScalarParameter() {defaultValue=2, parameterName=Param_10}
   ))
  n9 = Power(
   Base=MaterialFunctionCall(
    UVs=@n4
    Size=ScalarParameter() {defaultValue=0.1, parameterName=Param_4}
    GlowMin=ScalarParameter() {parameterName=Param_6}
    GlowMax=ScalarParameter() {defaultValue=1, parameterName=Param_5}
   ) {materialFunction=&MF_UI_Circle}.Fill
   Exp=ScalarParameter() {defaultValue=2, parameterName=Param_8}
  ) {constExponent=1}
  n10 = ScalarParameter() {defaultValue=0.075, parameterName=Param_4asdca}
  n11 = ScalarParameter() {parameterName=Param_6sdfvzxcv}
  n12 = ScalarParameter() {defaultValue=1, parameterName=Param_5zxcvzsdv}
  n13 = LinearInterpolate(
   A=LinearInterpolate(
    A=LinearInterpolate(A=@n2.RGB, B=@n3.RGB, Alpha=@n8)
    B=VectorParameter() {parameterName=RGB-3, defaultValue=<rgba 0 0.3061 0.8906 1>}.RGB
    Alpha=Saturate(None=@n9)
   )
   B=VectorParameter() {parameterName=RGB-4, defaultValue=<rgba 0.5955 0.8148 1 1>}.RGB
   Alpha=Power(
    Base=MaterialFunctionCall(UVs=@n4, Size=@n10, GlowMin=@n11, GlowMax=@n12) {materialFunction=&MF_UI_Circle}.Fill
    Exp=ScalarParameter() {defaultValue=8, parameterName=Param_8zxcve}
   ) {constExponent=1}
  )
  n14 = Constant() {r=1}
  n15 = MaterialFunctionCall(Float3=VectorParameter() {parameterName=Scene_Terrain_Target, defaultValue=<rgba 0 1 0 0>}.RGB) {materialFunction=&BreakOutFloat3Components}
  n16 = LinearInterpolate(
   A=LinearInterpolate(B=ScalarParameter() {defaultValue=0.45, parameterName=SceneDepth_BR}, Alpha=@n15.R)
   B=ScalarParameter() {defaultValue=0.7, parameterName=SceneDepth_Delmar}
   Alpha=@n15.G
  )
  n17 = ScalarParameter() {defaultValue=0.4, parameterName=SceneDepth_Juno}
  n18 = Subtract(
   A=Multiply(A=SceneTexture() {sceneTextureId=PPI_SceneDepth}.Color, B=ScalarParameter() {defaultValue=0.001, parameterName=SceneDepth-Normalize})
   B=LinearInterpolate(B=LinearInterpolate(A=@n16, B=@n17, Alpha=@n15.B), Alpha=ScalarParameter() {parameterName=Scene_Discovery_Transition}) {constA=2}
  )
  LinearInterpolate(
   A=@n1.Color
   B=LinearInterpolate(A=@n1.Color, B=AppendVector(A=@n13, B=@n14), Alpha=Saturate(None=@n18))
   Alpha=ScalarParameter() {parameterName=Scene_Discovery_Transition}
  )
 ]
}
