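# Post process material for the cosmetic turntable that writes only alpha before depth of field, adjusting it for separate translucency.
# It converts the SeparateTranslucency color to HSV and takes the value channel as brightness, alongside the former pass's alpha.
# Nested If nodes compute one minus brightness where translucency is above the default threshold and prior alpha is at least one, otherwise keeping the prior alpha.
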Material /CosmeticTurntable/Materials/PPM_TurntableSeparateTranslucence
graph {
 properties {
  bCastRayTracedShadows=false
  bUseTranslucencyVertexFog=false
  bWriteOnlyAlpha=true
  blendableLocation=BL_SceneColorBeforeDOF
  blendableOutputAlpha=true
  materialDomain=MD_PostProcess
  pixelDepthOffsetMode=PDOM_Legacy
  usedShadingModels=Default Lit
 }
 nodes [
  n1 = MaterialFunctionCall(Float3=MaterialFunctionCall!(RGB=SceneTexture() {desc=SeparateTranslucency, sceneTextureId=PPI_PostProcessInput1}.Color) {materialFunction=/Engine/Functions/Engine_MaterialFunctions02/Math/RGBtoHSV}."RGB to HSV (0-1)") {materialFunction=&BreakOutFloat3Components}
  n2 = ComponentMask(None=SceneTexture() {desc=FormerPassColor, sceneTextureId=PPI_PostProcessInput0}.Color) {a=true}
  n3 = OneMinus(None=@n1.B)
  If(A=@n1.B, "A > B"=If(A=@n2, "A > B"=@n3, "A == B"=@n3, "A < B"=@n2) {constB=1}, "A == B"=@n2, "A < B"=@n2)
 ]
}
